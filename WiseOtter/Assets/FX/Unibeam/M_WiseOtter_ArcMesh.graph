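# Masked, lit material for Niagara mesh particles in the WiseOtter Unibeam effect, using the HighTowerTomato gauntlet textures.
# A dissolve texture is sampled with offset UVs snapped to a coarse grid for a blocky look, then height-lerped by the Dissolve dynamic parameter.
# The inverted result becomes a bright blue glow, added to the gauntlet emissive, tinted by particle color and passed through MF_FX_EmissiveAdjustments.
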Material /WiseOtter/Assets/FX/Unibeam/M_WiseOtter_ArcMesh
graph {
 properties {
  bAutomaticallySetUsageInEditor=false
  bUsedWithNiagaraMeshParticles=true
  blendMode=BLEND_Masked
  previewMesh=/WiseOtter/Assets/Meshes/SM_WiseOtter_Unibeam
  usedShadingModels=Default Lit
 }
 nodes [
  Constant3Vector() {constant=<rgba 1 1 1 0>}
  LinearInterpolate(Alpha=MaterialFunctionCall() {materialFunction=&TimeWithSpeedVariable}.Frac Time)
  n1 = TextureSample() {texture=/WiseOtter/Assets/Textures/T_Gauntlet_HighTowerTomato_E}
  n2 = DynamicParameter() {paramNames [Dissolve,Param2,Param3,Param4]}
  n3 = Constant() {r=20}
  n4 = Multiply(A=Add(A=TextureCoordinate(), B=Constant2Vector() {g=-0.1, r=-0.2}), B=@n3) {constB=5}
  n5 = TextureSample(UVs=Divide(A=Floor(None=@n4), B=@n3)) {texture=/Game/CharClasses/Outlander/Gadgets/AntiMaterialCharge/Materials/T_AntiMaterialCharge_Dissolve_01}
  n6 = Multiply(
   A=Constant3Vector() {constant=<rgba 0 3.3333 10 1>}
   B=OneMinus(None=MaterialFunctionCall(Transition Phase=@n2.Dissolve, Height Texture=@n5.R) {materialFunction=&HeightLerp}.Alpha)
  )
  MaterialFunctionCall(Emissive=Multiply(A=Add(A=@n1.RGB, B=@n6), B=ParticleColor().RGB)) {materialFunction=&MF_FX_EmissiveAdjustments}
  TextureSample() {texture=/WiseOtter/Assets/Textures/T_Gauntlet_HighTowerTomato_D}
  TextureSample() {samplerType=SAMPLERTYPE_Normal, texture=/WiseOtter/Assets/Textures/T_Gauntlet_HighTowerTomato_N}
  TextureSample() {samplerType=SAMPLERTYPE_Masks, texture=/WiseOtter/Assets/Textures/T_Gauntlet_HighTowerTomato_S}
  TextureSample() {texture=/WiseOtter/Assets/Textures/T_Gauntlet_HighTowerTomato_M}
  WorldPosition()
 ]
}
