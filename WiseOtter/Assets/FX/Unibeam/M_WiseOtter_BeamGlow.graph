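# Translucent unlit glow material for Niagara mesh particles and ribbons, in the Unibeam FX set.
# A SphereMask with U tiling zeroed forms a soft band across V, eroded by the Dissolve dynamic parameter.
# Sine curves on U and V shape a saturated mask that tapers both ends.
# These masks multiply particle alpha into a parameterized DepthFade, while particle color feeds MF_FX_EmissiveAdjustments for emissive.
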Material /WiseOtter/Assets/FX/Unibeam/M_WiseOtter_BeamGlow
graph {
 properties {
  bAutomaticallySetUsageInEditor=false
  bUsedWithNiagaraMeshParticles=true
  bUsedWithNiagaraRibbons=true
  blendMode=BLEND_Translucent
  shadingModel=MSM_Unlit
  usedShadingModels=Unlit
 }
 nodes [
  n1 = ParticleColor()
  n2 = TextureCoordinate() {uTiling=0}
  n3 = Constant2Vector() {g=0.5}
  n4 = Subtract(
   A=Power(Base=SphereMask(A=@n2, B=@n3) {attenuationRadius=0.5, hardnessPercent=0}) {constExponent=1.5}
   B=DynamicParameter() {defaultValue=<rgba 2 0 1 1>, paramNames [UScale,Dissolve,Param3,Param4]}.Dissolve
  )
  n5 = Sine(None=ComponentMask(None=TextureCoordinate()) {g=true}) {period=2}
  n6 = ComponentMask(None=TextureCoordinate()) {r=true}
  n7 = LinearInterpolate(Alpha=OneMinus(None=Sine(None=@n6) {period=2})) {constA=-2}
  n8 = Saturate(None=Multiply(A=Subtract(A=@n5, B=@n7) {constB=0}) {constB=5})
  DepthFade(Opacity=Multiply(A=@n1.A, B=Multiply(A=@n4, B=@n8)), FadeDistance=ScalarParameter() {parameterName=FadeDistance}) {fadeDistanceDefault=0}
  MaterialFunctionCall(Emissive=@n1.RGB) {materialFunction=&MF_FX_EmissiveAdjustments}
 ]
}
