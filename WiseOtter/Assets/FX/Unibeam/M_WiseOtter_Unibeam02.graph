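# Translucent unlit Unibeam effect material for Niagara ribbons and mesh particles.
# Emissive adds two panning trail layers, one mapped through a color curve atlas, the other tinted by smoke-driven hue shift.
# Opacity combines a vertical SphereMask falloff, reduced by Dissolve, with sine-shaped UV edge tapering, particle alpha and DepthFade.
# Dynamic parameters supply Dissolve and UScale, which stretches texture tiling along the beam's U axis.
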Material /WiseOtter/Assets/FX/Unibeam/M_WiseOtter_Unibeam02
graph {
 properties {
  bAutomaticallySetUsageInEditor=false
  bUsedWithNiagaraMeshParticles=true
  bUsedWithNiagaraRibbons=true
  blendMode=BLEND_Translucent
  shadingModel=MSM_Unlit
  usedShadingModels=Unlit
 }
 nodes [
  n1 = AppendVector(A=DynamicParameter() {defaultValue=<rgba 2 0 1 1>, paramNames [UScale,Dissolve,Param3,Param4]}.UScale, B=Constant() {r=1})
  n2 = Multiply(A=TextureCoordinate() {uTiling=1.5}, B=@n1)
  n3 = CurveAtlasRowParameter(CurveTime=TextureSample(UVs=Panner(Coordinate=@n2) {speedX=-1.5}) {texture=/Game/Weapons/FORT_Melee/Pickaxe_Future_Samurai_Summer_Male/FX/T_Pickaxe_Future_Samurai_Summer_Male_Trail_03}.R) {
   atlas=&/CA_WiseOtter
   curve=&/CV_WiseOtter_UnibeamBlue
   defaultValue=2
   parameterName=ColorCV
  }
  Add(A=@n3) {constB=0}
  n4 = TextureCoordinate() {uTiling=0}
  n5 = Constant2Vector() {g=0.5}
  n6 = Subtract(
   A=Power(Base=SphereMask(A=@n4, B=@n5) {attenuationRadius=0.5, hardnessPercent=0}) {constExponent=1.5}
   B=DynamicParameter() {defaultValue=<rgba 2 0 1 1>, paramNames [UScale,Dissolve,Param3,Param4]}.Dissolve
  )
  n7 = Sine(None=ComponentMask(None=TextureCoordinate()) {g=true}) {period=2}
  n8 = ComponentMask(None=TextureCoordinate()) {r=true}
  n9 = LinearInterpolate(Alpha=OneMinus(None=Sine(None=@n8) {period=2})) {constA=-2}
  n10 = Saturate(None=Multiply(A=Subtract(A=@n7, B=@n9) {constB=0}) {constB=5})
  DepthFade(Opacity=Multiply(A=ParticleColor().A, B=Multiply(A=@n6, B=@n10)), FadeDistance=ScalarParameter() {parameterName=FadeDistance})
  n11 = DynamicParameter() {defaultValue=<rgba 2 0 1 1>, paramNames [UScale,Dissolve,Param3,Param4]}
  n12 = Constant() {r=1}
  n13 = TextureSample(UVs=Multiply(A=TextureCoordinate() {uTiling=0.7, vTiling=2}, B=AppendVector(A=@n11.UScale, B=@n12))) {texture=&T_Inky_Smoke_Tile}
  n14 = MaterialFunctionCall(
   Hue Shift Percentage=Add(A=LinearInterpolate(Alpha=@n13.R) {constB=0.3}, B=ScalarParameter() {defaultValue=-0.15, parameterName=Color offset}) {constB=0}
   Texture=VectorParameter() {parameterName=HueShiftColor, defaultValue=<rgba 0.0257 0.3504 1 1>}.RGB
  ) {materialFunction=&HueShift}
  n15 = DynamicParameter() {defaultValue=<rgba 2 0 1 1>, paramNames [UScale,Dissolve,Param3,Param4]}
  n16 = Constant() {r=1}
  n17 = Panner(Coordinate=Multiply(A=TextureCoordinate() {uTiling=0.5}, B=AppendVector(A=@n15.UScale, B=@n16))) {speedX=-1}
  n18 = Multiply(
   A=Desaturation(None=Multiply(A=@n14.Result), Fraction=ScalarParameter() {defaultValue=-1, parameterName=Des})
   B=Power(Base=TextureSample(UVs=@n17) {texture=/Game/Weapons/FORT_Melee/Pickaxe_Future_Samurai_Summer_Male/FX/T_Pickaxe_Future_Samurai_Summer_Male_Trail_02}.G) {constExponent=0.7}
  )
  MaterialFunctionCall(Emissive=Add(A=Multiply(A=@n18) {constB=0.8}, B=Multiply(A=@n3)) {constB=0}) {materialFunction=&MF_FX_EmissiveAdjustments}
  n19 = TextureCoordinate() {uTiling=1.3}
  TextureSample(UVs=Panner(Coordinate=Multiply(A=@n19, B=@n1)) {speedX=-1}) {texture=/Game/Weapons/FORT_Melee/Pickaxe_Tower_Sentinel_Female/FX/T_TowerSentinel_trail}
 ]
}
