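# Niagara system for WiseOtter's Unibeam effect, drawing a beam from a sampled skeletal mesh position to the user-supplied end point.
# Beams lays three spawn-group ribbon layers along a Hermite spline, with per-layer width, color and alpha falloff.
# Flare adds layered sprites at the sampled origin, and ShellBurst adds a scaling, fading mesh burst.
# PP spawns a short-lived component-renderer particle on the mesh, gated by User.LocalView for local-view-only feedback.
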NiagaraSystem /WiseOtter/Assets/FX/Unibeam/NS_WiseOtter_Beam
dependencies {
 usedDataInterfaces [&NiagaraDataInterfaceCurve, &NiagaraDataInterfaceColorCurve, &NiagaraDataInterfaceSkeletalMesh]
 usedDynamicInputs [
  &VectorLength
  &Subtract_Vector
  &FindQuatBetweenTwoVectors
  /Niagara/DynamicInputs/Lerp/Lerp_Position
  &MakeFloatFromInt
 ]
 usedModules [
  &self:SystemSpawnScript.SystemScriptSource.SystemScriptGraph.NiagaraNodeAssignment_3.SetVariables_#1
  &SystemState
  &SampleSkeletalMesh
  &self:SystemSpawnScript.SystemScriptSource.SystemScriptGraph.NiagaraNodeAssignment_0.SetVariables_#2
  &self:SystemSpawnScript.SystemScriptSource.SystemScriptGraph.NiagaraNodeAssignment_1.SetVariables_#3
  &Timeline
 ]
}
emitters {
 Beams {
  emitterUpdateScript [
   &EmitterState()
   &SpawnBurst_Instantaneous(Spawn Count=20, Use Spawn Probability=false, UseLoopCountLimit=false)
   &SpawnBurst_Instantaneous(Spawn Count=20, Spawn Group=1, Use Spawn Probability=false, UseLoopCountLimit=false) {module=SpawnBurst_Instantaneous001}
   &SpawnBurst_Instantaneous(Spawn Count=20, Spawn Group=2, Use Spawn Probability=false, UseLoopCountLimit=false) {module=SpawnBurst_Instantaneous002}
  ]
  particleSpawnScript [
   /SetVariables(Particles.RibbonID=/Niagara/DynamicInputs/Meshes/MakeRibbonIDFromSpawnGroup()) {isSetParameters=true}
   &PartitionParticles(
    Number Of Partitions=1
    Particles Per Partition=20
    Partition Mode:ENiagara_PartitionMode=::NewEnumerator1
    Write Partition Index to Ribbon Link Order:ENiagara_AttributeSamplingApplyOutput=::NewEnumerator0
   )
   &InitializeParticle(
    Color=&MaskLinearColorBySpawnGroup(
     Masked Linear Color=<rgba 1 1 1 0.3>
     Passthrough Linear Color=&MaskLinearColorBySpawnGroup(Masked Linear Color=<rgba 0 0.2333 0.7 0.8>, Passthrough Linear Color=<rgba 1 1 1 0.15>, Spawn Group Mask=1)
    )
    Lifetime=&MaskFloatBySpawnGroup(Masked Float=0.3, Passthrough Float=0.5)
    Ribbon Width=&MaskFloatBySpawnGroup(Masked Float=140, Passthrough Float=90, Spawn Group Mask=1)
    Ribbon Width Mode=::NewEnumerator1
    UsePositionOffset=false
    Write Material Random=true
   )
  ]
  particleUpdateScript [
   &ParticleState()
   /Game/Effects/Niagara/Modules/Curves/FN_HermiteSpline(
    Position Blend:NiagaraFloat=$Particles.Initial.PartitionParticles.Normalized Index In Partition
    Position End:NiagaraPosition=$User.BeamEnd_0
    Position Start:Vector3f=&Add_Vector(
     A=Linked() {linkedVariable {name=System.SampleSkeletalMesh.SampledPosition, type=NiagaraPosition}}
     B=&RotateVectorByQuaternion(Vector To Rotate=<0 0 0>, Quaternion=$System.BeamOrientation)
    )
    Write Particles Position=true
   )
   /SetVariables(
    Particles.Beam Fade_Width:NiagaraFloat=&Multiply_Float(
     A=$Particles.Initial.RibbonWidth
     B=&MaskFloatBySpawnGroup(Masked Float=1, Passthrough Float=0, Spawn Group Mask=2)
    )
    Particles.BeamGlow_Width:NiagaraFloat=&Multiply_Float(
     A=$Particles.Initial.RibbonWidth
     B=&MaskFloatBySpawnGroup(Masked Float=1, Passthrough Float=0, Spawn Group Mask=1)
    )
    Particles.ProjectileBeam_Width:NiagaraFloat=&Multiply_Float(A=$Particles.Initial.RibbonWidth, B=&MaskFloatBySpawnGroup(Masked Float=1, Passthrough Float=0))
   ) {isSetParameters=true}
   &ScaleColor(
    Scale Alpha=&MaskFloatBySpawnGroup(
     Masked Float=&Multiply_Float(
      A=&FloatFromCurve(
       FloatCurve=DataInterface() {propertyValues {Curve {keys [
           {arriveTangent=0.1111, interpMode=RCIM_Cubic, leaveTangent=0.1111, leaveTangentWeight=0.7387, tangentMode=RCTM_User, time=0, value=1}
           {interpMode=RCIM_Cubic, tangentMode=RCTM_User, time=0.2, value=1}
           <at 0.8 1 RCIM_Cubic>
           <at 1 1>
          ]}}}
       CurveIndex=&Multiply_Float(A=$Particles.PartitionParticles.Normalized Index In Partition, B=1)
       Scale Curve=10
      )
      B=&FloatFromCurve(
       FloatCurve=DataInterface() {propertyValues {Curve {keys [
           {arriveTangent=0.1111, interpMode=RCIM_Cubic, leaveTangent=0.1111, leaveTangentWeight=0.7387, tangentMode=RCTM_User, time=0, value=1}
           <at 0.2 1>
           {arriveTangentWeight=0.2134, interpMode=RCIM_Cubic, leaveTangentWeight=0.2134, tangentMode=RCTM_User, time=0.6, value=1}
           <at 1 0>
          ]}}}
       CurveIndex=$Particles.PartitionParticles.Normalized Index In Partition
       Scale Curve=1.2
      )
     )
     Passthrough Float=&MaskFloatBySpawnGroup(
      Masked Float=&FloatFromCurve(
       FloatCurve=DataInterface() {propertyValues {Curve {keys [
           {arriveTangent=0.1111, interpMode=RCIM_Cubic, leaveTangent=0.1111, leaveTangentWeight=0.7387, tangentMode=RCTM_User, time=0, value=1}
           <at 0.2 1>
           {arriveTangentWeight=0.2134, interpMode=RCIM_Cubic, leaveTangentWeight=0.2134, tangentMode=RCTM_User, time=0.6, value=1}
           <at 1 0>
          ]}}}
       CurveIndex=$Particles.PartitionParticles.Normalized Index In Partition
      )
      Passthrough Float=&Multiply_Float(
       A=&Multiply_Float(
        A=10
        B=&FloatFromCurve(FloatCurve=DataInterface() {propertyValues {Curve {keys [
             {arriveTangent=-1.5385, interpMode=RCIM_Cubic, leaveTangent=-1.5385, tangentMode=RCTM_User, time=0, value=1}
             {arriveTangent=-2.602, interpMode=RCIM_Cubic, leaveTangent=-2.602, tangentMode=RCTM_User, time=1, value=0}
            ]}}})
       )
       B=&FloatFromCurve(
        FloatCurve=DataInterface() {propertyValues {Curve {keys [
            {arriveTangent=0.1111, interpMode=RCIM_Cubic, leaveTangent=0.1111, leaveTangentWeight=0.7387, tangentMode=RCTM_User, time=0, value=1}
            <at 0.2 1>
            {arriveTangentWeight=0.2134, interpMode=RCIM_Cubic, leaveTangentWeight=0.2134, tangentMode=RCTM_User, time=0.6, value=1}
            <at 1 0>
           ]}}}
        CurveIndex=$Particles.PartitionParticles.Normalized Index In Partition
        Scale Curve=1.2
       )
      )
      Spawn Group Mask=1
     )
    )
    Scale RGB=&VectorFromFloat(Value=&MaskFloatBySpawnGroup(Masked Float=2, Passthrough Float=1, Spawn Group Mask=2))
    ScaleA=true
    ScaleRGB=true
   )
   &DynamicMaterialParameters(
    Index 0 Param 1=&MaskFloatBySpawnGroup(Masked Float=2, Passthrough Float=1)
    Index 0 Param 2=&MaskFloatBySpawnGroup(
     Masked Float=&FloatFromCurve(FloatCurve=DataInterface() {propertyValues {Curve {keys [<at 0.3 0 RCIM_Cubic>, {arriveTangent=2.2097, tangentMode=RCTM_User, time=1, value=1}]}}})
     Passthrough Float=&MaskFloatBySpawnGroup(
      Masked Float=&FloatFromCurve(FloatCurve=DataInterface() {propertyValues {Curve {keys [<at 0.3 0 RCIM_Cubic>, {arriveTangent=2.2097, tangentMode=RCTM_User, time=1, value=0.9}]}}})
      Passthrough Float=&FloatFromCurve(FloatCurve=DataInterface() {propertyValues {Curve {keys [<at 0.3 0 RCIM_Cubic>, {arriveTangent=2.2097, tangentMode=RCTM_User, time=1, value=0.8}]}}})
      Spawn Group Mask=1
     )
    )
    Module.Index0Param0Write=true
    Module.Index0Param1Write=true
    Module.Index0Param2Write=false
    Module.Index0Param3Write=false
   )
  ]
  rendererClasses [&NiagaraRibbonRendererProperties]
  renderers [
   {rendererClass=&NiagaraRibbonRendererProperties}
   {rendererClass=&NiagaraRibbonRendererProperties, rendererIndex=1}
   {rendererClass=&NiagaraRibbonRendererProperties, rendererIndex=2}
  ]
 }
 Flare {
  emitterUpdateScript [
   &EmitterState(Life Cycle Mode=::NewEnumerator1, Loop Behavior=::NewEnumerator1, Loop Duration=1, UseLoopDelay=false)
   &SpawnBurst_Instantaneous(Use Spawn Probability=false, UseLoopCountLimit=false)
   &SpawnRate(
    SpawnRate=&Multiply_Float(
     A=10
     B=&FloatFromCurve(FloatCurve=DataInterface() {propertyValues {Curve {keys [<at 0 1 RCIM_Constant>,<at 0.25 0 RCIM_Constant>]}}}, CurveIndex=$Emitter.Age)
    )
    Use Spawn Probability=false
   )
   &SpawnBurst_Instantaneous(Spawn Group=1, Use Spawn Probability=false, UseLoopCountLimit=false) {module=SpawnBurst_Instantaneous001}
   &SpawnBurst_Instantaneous(Spawn Group=2, Use Spawn Probability=false, UseLoopCountLimit=false) {module=SpawnBurst_Instantaneous002}
   &SpawnBurst_Instantaneous(Spawn Group=3, Use Spawn Probability=false, UseLoopCountLimit=false) {module=SpawnBurst_Instantaneous003}
   &SpawnRate(
    SpawnGroup=3
    SpawnRate=&Multiply_Float(
     A=10
     B=&FloatFromCurve(FloatCurve=DataInterface() {propertyValues {Curve {keys [<at 0 1 RCIM_Constant>,<at 0.4 0 RCIM_Constant>]}}}, CurveIndex=$Emitter.Age)
    )
    Use Spawn Probability=false
   ) {module=SpawnRate001}
  ]
  particleSpawnScript [
   /SetVariables(Particles.VisibilityTag=&SpawnGroup()) {isSetParameters=true}
   &InitializeParticle(
    Color=&MaskLinearColorBySpawnGroup(
     Masked Linear Color=<rgba 2 2 2 5>
     Passthrough Linear Color=&MaskLinearColorBySpawnGroup(
      Masked Linear Color=<rgba 1 1 1 1>
      Passthrough Linear Color=&MaskLinearColorBySpawnGroup(Masked Linear Color=<rgba 0.276 1.8507 5 1>, Passthrough Linear Color=<rgba 1.5 1.5 1.5 2>, Spawn Group Mask=2)
      Spawn Group Mask=1
     )
    )
    Lifetime=&MaskFloatBySpawnGroup(
     Masked Float=0.7
     Passthrough Float=&MaskFloatBySpawnGroup(Masked Float=0.5, Passthrough Float=0.3, Spawn Group Mask=2)
     Spawn Group Mask=1
    )
    Sprite Rotation Mode=::NewEnumerator2
    Sprite Size Mode=::NewEnumerator3
    Uniform Sprite Size=&MaskFloatBySpawnGroup(
     Masked Float=70
     Passthrough Float=&MaskFloatBySpawnGroup(Masked Float=220, Passthrough Float=200, Spawn Group Mask=3)
     Spawn Group Mask=2
    )
    UsePositionOffset=false
    Write Material Random=true
   )
   &CameraOffset(Camera Offset Amount=&MaskFloatBySpawnGroup(Masked Float=0, Passthrough Float=-30, Spawn Group Mask=2))
  ]
  particleUpdateScript [
   &ParticleState()
   /SetVariables(Particles.Position=&AddVectorToPosition(
     Position=$System.SampleSkeletalMesh.SampledPosition
     Vector=&TransformVector(Vector=&Multiply_VectorByFloat(Vector=<20 0 10>, Float=1), Destination Space=::Simulation)
    )) {isSetParameters=true}
   &ScaleColor(
    Scale Alpha=&MaskFloatBySpawnGroup(
     Masked Float=&FloatFromCurve(
      FloatCurve=DataInterface() {propertyValues {Curve {keys [
          {interpMode=RCIM_Cubic, tangentMode=RCTM_Break, time=0, value=1}
          {arriveTangentWeight=0.7369, interpMode=RCIM_Cubic, tangentMode=RCTM_Break, time=1, value=0}
         ]}}}
      Scale Curve=2
     )
     Passthrough Float=&MaskFloatBySpawnGroup(
      Masked Float=&FloatFromCurve(
       FloatCurve=DataInterface() {propertyValues {Curve {keys [
           {interpMode=RCIM_Cubic, tangentMode=RCTM_Break, time=0.8, value=1}
           {arriveTangentWeight=0.7369, interpMode=RCIM_Cubic, tangentMode=RCTM_Break, time=1, value=0}
          ]}}}
       Scale Curve=0.12
      )
      Passthrough Float=&MaskFloatBySpawnGroup(
       Masked Float=&FloatFromCurve(
        FloatCurve=DataInterface() {propertyValues {Curve {keys [
            {interpMode=RCIM_Cubic, tangentMode=RCTM_User, time=0, value=0}
            {arriveTangent=-0.7556, interpMode=RCIM_Cubic, leaveTangent=-0.7556, tangentMode=RCTM_User, time=0.2, value=1}
            {arriveTangent=-2.5551, tangentMode=RCTM_User, time=1, value=0}
           ]}}}
        Scale Curve=0.5
       )
       Passthrough Float=&FloatFromCurve(
        FloatCurve=DataInterface() {propertyValues {Curve {keys [{interpMode=RCIM_Cubic, tangentMode=RCTM_User, time=0, value=1}, {interpMode=RCIM_Cubic, tangentMode=RCTM_User, time=1, value=0}]}}}
        Scale Curve=2
       )
       Spawn Group Mask=2
      )
      Spawn Group Mask=1
     )
    )
    Scale RGB=&MakeVectorFromLinearColorRGB(Color=&MaskLinearColorBySpawnGroup(
      Masked Linear Color=&ColorFromCurve(ColorCurve=DataInterface() {propertyValues {
         AlphaCurve {keys []}
         BlueCurve {keys [<at 0.1796 20 RCIM_Cubic>,<at 0.417 5>,<at 0.6208 0.7083>]}
         GreenCurve {keys [<at 0.1796 5.6 RCIM_Cubic>,<at 0.417 2.3688>,<at 0.6208 0.1469>]}
         RedCurve {keys [<at 0.1796 0.8 RCIM_Cubic>,<at 0.417 0>,<at 0.6208 0.0329>]}
        }})
      Passthrough Linear Color=&MaskLinearColorBySpawnGroup(
       Masked Linear Color=&ColorFromCurve(ColorCurve=DataInterface() {propertyValues {
          AlphaCurve {keys []}
          BlueCurve {keys [<at 0.1796 10 RCIM_Cubic>,<at 0.7694 0.2135>]}
          GreenCurve {keys [<at 0.1796 2.5 RCIM_Cubic>,<at 0.7694 0>]}
          RedCurve {keys [<at 0.1796 0 RCIM_Cubic>,<at 0.7694 0.0356>]}
         }})
       Passthrough Linear Color=&MaskLinearColorBySpawnGroup(
        Masked Linear Color=<rgba 1 1 1 1>
        Passthrough Linear Color=&ColorFromCurve(ColorCurve=DataInterface() {propertyValues {
           AlphaCurve {keys []}
           BlueCurve {keys [<at 0.0386 6 RCIM_Cubic>, <at 0.4137 2>, {interpMode=RCIM_Cubic, leaveTangent=-0.7175, time=0.8388, value=1}]}
           GreenCurve {keys [<at 0.0386 3.2 RCIM_Cubic>, <at 0.4137 1>, {interpMode=RCIM_Cubic, leaveTangent=-4.4353, time=0.8388, value=0.1667}]}
           RedCurve {keys [<at 0.0386 1.8 RCIM_Cubic>, <at 0.4137 0>, {interpMode=RCIM_Cubic, leaveTangent=2.2176, time=0.8388, value=0}]}
          }})
        Spawn Group Mask=2
       )
       Spawn Group Mask=1
      )
     ))
    ScaleA=true
    ScaleRGB=true
   )
   &ScaleSpriteSize(
    Scale Factor=&MaskVector2dBySpawnGroup(
     Masked Vector2D=&Vector2DFromFloat(Value=&FloatFromCurve(
       FloatCurve=DataInterface() {propertyValues {Curve {keys [
           {arriveTangent=1.6411, interpMode=RCIM_Cubic, leaveTangent=1.6411, tangentMode=RCTM_User, time=0, value=0.5}
           {arriveTangent=1, interpMode=RCIM_Cubic, leaveTangent=1, time=0.3211, value=0.65}
           {tangentMode=RCTM_User, time=1, value=1.5}
          ]}}}
       Scale Curve=1.2
      ))
     Passthrough Vector2D=&MaskVector2dBySpawnGroup(
      Masked Vector2D=&Vector2DFromFloat(Value=&FloatFromCurve(
        FloatCurve=DataInterface() {propertyValues {Curve {keys [{interpMode=RCIM_Cubic, tangentMode=RCTM_User, time=0.7, value=1}, {interpMode=RCIM_Cubic, tangentMode=RCTM_User, time=1, value=0}]}}}
        Scale Curve=0.9
       ))
      Passthrough Vector2D=&MaskVector2dBySpawnGroup(
       Masked Vector2D=&Vector2DFromFloat(Value=&FloatFromCurve(
         FloatCurve=DataInterface() {propertyValues {Curve {keys [
             <at 0 0 RCIM_Cubic>
             {arriveTangent=4, interpMode=RCIM_Cubic, leaveTangent=4, time=0.2, value=4}
             {arriveTangent=-12.0286, interpMode=RCIM_Cubic, leaveTangent=-12.0286, tangentMode=RCTM_User, time=1, value=4}
            ]}}}
         Scale Curve=&RandomRangeFloat(Minimum=0.2, Maximum=1, Evaluation Type=::NewEnumerator1)
        ))
       Passthrough Vector2D=&Vector2DFromFloat(Value=&FloatFromCurve(
         FloatCurve=DataInterface() {propertyValues {Curve {keys [
             <at 0 0 RCIM_Cubic>
             {arriveTangent=1, interpMode=RCIM_Cubic, leaveTangent=1, time=0.3659, value=0.7136}
             {tangentMode=RCTM_User, time=1, value=1}
            ]}}}
         Scale Curve=&Multiply_Float(A=0.5, B=1)
        ))
       Spawn Group Mask=2
      )
      Spawn Group Mask=1
     )
     Spawn Group Mask=0
    )
    Scale Sprite Size Mode=::NewEnumerator2
   )
   &DynamicMaterialParameters(
    Index 0 Param 1=&MaskFloatBySpawnGroup(
     Masked Float=&FloatFromCurve(FloatCurve=DataInterface() {propertyValues {Curve {keys [{tangentMode=RCTM_User, time=0, value=4}, {tangentMode=RCTM_User, time=1, value=5}]}}})
     Passthrough Float=&MaskFloatBySpawnGroup(
      Masked Float=&FloatFromCurve(FloatCurve=DataInterface() {propertyValues {Curve {keys [{tangentMode=RCTM_User, time=0, value=4}, {tangentMode=RCTM_User, time=1, value=5}]}}})
      Passthrough Float=1
      Spawn Group Mask=1
     )
    )
    Module.Index0Param0Write=true
    Module.Index0Param1Write=true
    Module.Index0Param2Write=false
    Module.Index0Param3Write=false
   )
  ]
  rendererClasses [&NiagaraSpriteRendererProperties]
  renderers [
   {rendererClass=&NiagaraSpriteRendererProperties}
   {rendererClass=&NiagaraSpriteRendererProperties, rendererIndex=1}
   {rendererClass=&NiagaraSpriteRendererProperties, rendererIndex=2}
   {rendererClass=&NiagaraSpriteRendererProperties, rendererIndex=3}
  ]
 }
 PP {
  emitterUpdateScript [
   &EmitterState(Life Cycle Mode=::NewEnumerator1, Loop Behavior=::NewEnumerator1, Loop Duration=1, UseLoopDelay=false)
   &SpawnBurst_Instantaneous(
    Spawn Probability=&MakeCustomFloatFromBool(Bool=$User.LocalView, True Float=1, False Float=0)
    Use Spawn Probability=true
    UseLoopCountLimit=false
   )
  ]
  particleSpawnScript [&InitializeParticle(
    Color Mode=::NewEnumerator0
    Lifetime=0.5
    Sprite Size Mode=::NewEnumerator3
    Uniform Sprite Size=200
    UsePositionOffset=false
    Write Material Random=true
   )]
  particleUpdateScript [
   /SetVariables(Particles.SpeedNormalizedB:NiagaraFloat=&FloatFromCurve(FloatCurve=DataInterface() {propertyValues {Curve {keys [{interpMode=RCIM_Cubic, tangentMode=RCTM_User, time=0.2, value=0.8}, {interpMode=RCIM_Cubic, tangentMode=RCTM_User, time=1, value=0}]}}})) {isSetParameters=true}
   &SkeletalMeshLocation(MaskSpawnGroup=false, Mesh Sampling Type=::NewEnumerator5, Skeletal Mesh=$System.SK_Mesh, Socket Sampling Mode=::NewEnumerator5)
   &ParticleState()
   /SetVariables(Particles.ComponentsEnabled:NiagaraBool=&SetBoolByFloatComparison(
     A=$Particles.Age
     B=&Subtract_Float(A=$Particles.Lifetime, B=$Engine.DeltaTime)
     Comparison Type=::NewEnumerator6
    )) {isSetParameters=true}
  ]
  rendererClasses [&NiagaraComponentRendererProperties]
  renderers [{rendererClass=&NiagaraComponentRendererProperties, rendererIndex=0}]
 }
 ShellBurst {
  emitterUpdateScript [
   &EmitterState(Life Cycle Mode=::NewEnumerator1, Loop Behavior=::NewEnumerator1, Loop Delay=1.35, Loop Duration=1, UseLoopDelay=false)
   &SpawnBurst_Instantaneous(Use Spawn Probability=false, UseLoopCountLimit=false)
  ]
  particleSpawnScript [
   &InitializeParticle(
    Color=<rgba 0 2 2 1>
    Lifetime=0.5
    Mesh Scale=<0.8 0.8 0.8>
    Mesh Scale Mode=::NewEnumerator4
    Position=<15 0 5>
    Position Mode=::NewEnumerator1
    UsePositionOffset=false
    Write Material Random=true
   )
   &InitialMeshOrientation(Mesh Orientation Mode=::NewEnumerator2, Rotation=<0 0 0.5>, Use Rotation Vector=true)
  ]
  particleUpdateScript [
   &ParticleState()
   &ScaleMeshSize(Scale Factor=&VectorFromFloat(Value=&FloatFromCurve(FloatCurve=DataInterface() {propertyValues {Curve {keys [<at 0 0.8>,<at 1 1>]}}}, Scale Curve=0.6)))
   &DynamicMaterialParameters(
    Index 0 Param 1=0.25
    Index 0 Param 2=-1
    Index 0 Param 3=1
    Index 0 Param 4=1
    Module.Index0Param0Write=true
    Module.Index0Param1Write=true
    Module.Index0Param2Write=true
    Module.Index0Param3Write=true
   )
   &ScaleColor(
    Scale Alpha=&FloatFromCurve(FloatCurve=DataInterface() {propertyValues {Curve {keys [<at -0.0007 0 RCIM_Constant>,<at 0.0999 1.0037 RCIM_Cubic>,<at 1 0>]}}})
    ScaleA=true
    ScaleRGB=true
   )
   /SetVariables(Particles.Position=&AddVectorToPosition(
     Position=&Lerp_Vector(A=$System.SampleSkeletalMesh.SampledPosition, B=$System.SampleSkeletalMesh001.SampledPosition, Alpha=0)
     Vector=&TransformVector(Vector=&Multiply_VectorByFloat(Vector=<20 0 10>, Float=1), Destination Space=::Simulation)
    )) {enabled=false, isSetParameters=true}
  ]
  rendererClasses [&NiagaraMeshRendererProperties]
  renderers [{rendererClass=&NiagaraMeshRendererProperties, rendererIndex=0}]
 }
}
summary {
 systemName=NS_WiseOtter_Beam
 emitters [
  {emitterName=PP, rendererClasses [&NiagaraComponentRendererProperties]}
  {emitterName=ShellBurst, rendererClasses [&NiagaraMeshRendererProperties]}
  {emitterName=Beams, rendererClasses [&NiagaraRibbonRendererProperties]}
  {emitterName=Flare, rendererClasses [&NiagaraSpriteRendererProperties]}
 ]
 userVariables [
  User.BeamEnd_0 {type=NiagaraPosition, description="", defaultValue {}}
  User.BeamFadeOut {type=NiagaraInt32}
  User.LocalView {type=NiagaraBool}
  User.Scale {type=NiagaraFloat, defaultValue=0.2, description=""}
 ]
}
